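# Niagara system for the StarStray character emote: glitchy, color-cycling CD and disk effects attached to skeletal mesh sockets.
# An anticipation sprite spirals toward a socket for the User LoopDelay, after which lit CDs, layered mesh disks, a flare and glitch squares appear.
# Notes_Line draws three camera-plane ribbon rings with curve-driven waveform spikes.
# CD and anticipation particles emit location events, and User EffectScale sizes several emitters.
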NiagaraSystem /BRCosmetics/Animation/Game/MainPlayer/Emotes/StarStray/FX/NS_Emote_StarStray
dependencies {
 usedDataInterfaces [
  &NiagaraDataInterfaceArrayFloat
  &NiagaraDataInterfaceCurve
  &NiagaraDataInterfaceColorCurve
  &NiagaraDataInterfaceVector2DCurve
  &NiagaraDataInterfaceVectorCurve
  &NiagaraDataInterfaceArrayColor
  &NiagaraDataInterfaceCamera
 ]
 usedDynamicInputs [&SpawnGroup, &VectorFromCurve, &ClampFloat]
 usedModules [
  &SystemState
  &self:SystemSpawnScript.SystemScriptSource.SystemScriptGraph.NiagaraNodeAssignment_1.SetVariables_#1
  /Game/Effects/Niagara/Modules/Audio/FN_NPC_EventAudio
 ]
}
emitters {
 Anticipation_Square {
  emitterUpdateScript [
   &EmitterState(Life Cycle Mode=::NewEnumerator1, Loop Behavior=::NewEnumerator1, Loop Duration Mode=::NewEnumerator1, UseLoopDelay=false)
   &SpawnBurst_Instantaneous(Spawn Count=1, Use Spawn Probability=true)
  ]
  particleSpawnScript [
   &InitializeParticle(
    Color=&SelectLinearColorFromArray(
     Array Sampling Mode=::NewEnumerator0
     Direct Array Index=?
     Color Selection Array=DataInterface() {propertyValues {ColorData [<rgba 1.5 0.0855 1.0027 1>, <rgba 0.076 0.7671 1 1>, <rgba 0.114 1 0.3468 1>, <rgba 0.4293 0.159 1.5 1>]}}
     Random Seed=0
    )
    Lifetime=$User.LoopDelay
    Sprite Size Mode=::NewEnumerator3
    Uniform Sprite Size=3
    UsePositionOffset=false
    Write Material Random=true
   )
   &SkeletalMeshLocation(
    Bone / Socket Index=2
    Bone Sampling Mode=::NewEnumerator4
    Interpolate Mesh Sampling=false
    MaskSpawnGroup=false
    Orientation Sampling=::NewEnumerator1
    Skeletal Mesh=$User.Skeletal Mesh
   )
   &DynamicMaterialParameters(
    Index 0 Param 1=0.5
    Module.Param0WriteEnabled=true
    Module.Param1WriteEnabled=true
    Module.Param2WriteEnabled=true
    Module.Param3WriteEnabled=true
   )
  ]
  particleUpdateScript [
   &SolveForcesAndVelocity(Clamp Velocity=false, Limit Acceleration=false)
   &ParticleState()
   &SkeletalMeshLocation(
    Bone / Socket Index=2
    Bone Sampling Mode=::NewEnumerator4
    MaskSpawnGroup=false
    Orientation Sampling=::NewEnumerator1
    Position Sampling=::NewEnumerator2
    Skeletal Mesh=$User.Skeletal Mesh
   ) {module=SkeletalMeshLocation001}
   &SkeletalMeshLocation(
    Bone / Socket Index=3
    Bone Sampling Mode=::NewEnumerator4
    MaskSpawnGroup=false
    Orientation Sampling=::NewEnumerator1
    Position Sampling=::NewEnumerator2
    Skeletal Mesh=$User.Skeletal Mesh
   ) {module=SkeletalMeshLocation002}
   &RotateAroundPoint(
    Radius:NiagaraFloat=&FloatFromCurve(
     FloatCurve=DataInterface() {propertyValues {Curve {keys [
         {arriveTangent=3.6795, interpMode=RCIM_Cubic, leaveTangent=3.6795, tangentMode=RCTM_User, time=0, value=0}
         {interpMode=RCIM_Cubic, tangentMode=RCTM_User, time=0.8, value=1}
        ]}}}
     Scale Curve=30
    )
    Rotation Center:NiagaraPosition=$Particles.Initial.Position
    Rotation Phase=0.5
    Rotation Rate=-2.5
    X Axis:Vector3f=&RotateVectorByQuaternion(Vector To Rotate=<0 1 0>, Quaternion=$Particles.SkeletalMeshLocation001.SampledRotation)
    Y Axis:Vector3f=&RotateVectorByQuaternion(Vector To Rotate=<0 0 1>, Quaternion=$Particles.SkeletalMeshLocation001.SampledRotation)
   )
   &ScaleSpriteSize(Uniform Curve Sprite Scale=DataInterface() {propertyValues {Curve {keys [
        {arriveTangent=2.9798, interpMode=RCIM_Cubic, leaveTangent=2.9798, tangentMode=RCTM_User, time=0, value=0}
        {arriveTangent=-0.0245, interpMode=RCIM_Cubic, leaveTangent=-0.0245, tangentMode=RCTM_User, time=0.8, value=1}
        <at 1 0.45>
       ]}}})
   &SpriteRotationRate(Rotation Rate=&RandomRangeFloat(Minimum=200, Maximum=300))
   &ScaleColor(
    Scale Alpha=&FloatFromCurve(FloatCurve=DataInterface() {propertyValues {Curve {keys [
         {arriveTangent=4.1785, interpMode=RCIM_Cubic, leaveTangent=4.1785, tangentMode=RCTM_User, time=0, value=0}
         {arriveTangent=1.8433, interpMode=RCIM_Cubic, leaveTangent=1.8433, tangentMode=RCTM_User, time=0.2, value=0.65}
         {arriveTangent=0.0224, interpMode=RCIM_Cubic, leaveTangent=0.0224, tangentMode=RCTM_User, time=0.7, value=1}
        ]}}})
    Scale RGB=&VectorFromFloat(Value=&FloatFromCurve(
      FloatCurve=DataInterface() {propertyValues {Curve {keys [
          {arriveTangent=9.0725, interpMode=RCIM_Cubic, leaveTangent=9.0725, tangentMode=RCTM_User, time=0, value=0.75}
          {arriveTangent=-2.5249, interpMode=RCIM_Cubic, leaveTangent=-2.5249, tangentMode=RCTM_User, time=0.1408, value=0.9847}
          {arriveTangent=-0.0405, interpMode=RCIM_Cubic, leaveTangent=-0.0405, tangentMode=RCTM_User, time=1, value=0}
         ]}}}
      Scale Curve=5
     ))
    ScaleA=true
    ScaleRGB=false
   )
   &CameraOffset(Camera Offset Amount=5) {enabled=false, module=CameraOffset001}
   /SetVariables(Particles.Position=&Lerp_Vector(A=$Output.RotateAroundPoint.OutputPosition, B=$Particles.SkeletalMeshLocation002.SampledPosition, Alpha=$Particles.NormalizedAge)) {isSetParameters=true}
   &GenerateLocationEvent(Event Send Rate=60, Use Event Probability=false, UseEventDelay=false)
   &FN_Sprite_Attributes(
    Color=&MultiplyLinearColor(Linear Color=$Particles.Color, Scale Factor=<1 1 1 1>)
    Create Color Attribute=true
    Create SpriteSize Attribute=true
    SpriteSize=&Multiply_Vector2D_ByFloat(Float=1.5, Vector2D=$Particles.SpriteSize)
    Write to Module Namespace=true
   )
  ]
  rendererClasses [&NiagaraSpriteRendererProperties]
  renderers [
   {rendererClass=&NiagaraSpriteRendererProperties}
   {rendererClass=&NiagaraSpriteRendererProperties, rendererIndex=1}
  ]
 }
 CD {
  emitterUpdateScript [
   &EmitterState(
    Inactive Response=::NewEnumerator1
    Life Cycle Mode=::NewEnumerator1
    Loop Behavior=::NewEnumerator1
    Loop Delay=$User.LoopDelay
    Loop Duration Mode=::NewEnumerator1
    UseLoopDelay=true
   )
   &SpawnBurst_Instantaneous(Spawn Count=2, Use Spawn Probability=false, UseLoopCountLimit=false)
  ]
  particleSpawnScript [
   &FN_SetSpawnGroup(Number of Spawn Groups=2)
   &InitializeParticle(
    Color=<rgba 5 5 5 0.1>
    Lifetime=10
    Position Offset:NiagaraPosition=$User.Position Offset
    Sprite Size Mode=::NewEnumerator3
    Uniform Sprite Size=160
    UsePositionOffset=true
    Write Material Random=true
   )
   &InitialMeshOrientation(Mesh Orientation Mode=::NewEnumerator2, Orientation Vector=<0 0 1>, Use Rotation Vector=true)
  ]
  particleUpdateScript [
   &ParticleState(Kill Particles When Lifetime Has Elapsed=false)
   &SkeletalMeshLocation(
    Bone / Socket Index=$Particles.SpawnGroup
    Bone Sampling Mode=::NewEnumerator4
    MaskSpawnGroup=false
    Orientation Sampling=::NewEnumerator1
    Sampled Position Offset:NiagaraPosition=$User.Position Offset
    Skeletal Mesh=$User.Skeletal Mesh
   )
   &ScaleSpriteSize(
    Uniform Curve Index=$Particles.Age
    Uniform Curve Scale=$User.EffectScale
    Uniform Curve Sprite Scale=DataInterface() {propertyValues {Curve {keys [
        {arriveTangent=0.0068, interpMode=RCIM_Cubic, leaveTangent=0.0068, tangentMode=RCTM_User, time=0, value=0.35}
        {arriveTangent=0.0192, interpMode=RCIM_Cubic, leaveTangent=0.0192, tangentMode=RCTM_User, time=1.2, value=1}
       ]}}}
   )
   &SolveForcesAndVelocity(Clamp Velocity=false, Limit Acceleration=false)
   &UpdateMeshOrientation(
    Rotation Coordinate Space=::NewEnumerator0
    Rotation Rate=&FloatFromCurve(
     FloatCurve=DataInterface() {propertyValues {Curve {keys [
         {interpMode=RCIM_Cubic, tangentMode=RCTM_User, time=0, value=0.25}
         {interpMode=RCIM_Cubic, tangentMode=RCTM_User, time=0.2, value=0.875}
         {arriveTangent=-0.0568, interpMode=RCIM_Cubic, leaveTangent=-0.0568, tangentMode=RCTM_User, time=0.5, value=-0.5}
         {arriveTangent=0.0228, interpMode=RCIM_Cubic, leaveTangent=0.0228, tangentMode=RCTM_User, time=0.8, value=0.375}
         <at 1 0.25 RCIM_Cubic>
        ]}}}
     CurveIndex=&Frac_Float(Float=$Particles.NormalizedAge)
     Scale Curve=-1
    )
    Rotation Vector=&MakeVector(X=0, Y=0, Z=1)
   )
   &AlignSpriteToMeshOrientation(
    Mesh Orientation Relative Sprite Facing Vector:Vector3f=&RotateVectorByQuaternion(
     Vector To Rotate=<1 0 0>
     Quaternion=&MakeQuatFromAxisAngle(
      Axis=<0 1 1>
      AngleInDegrees=&FloatFromCurve(
       FloatCurve=DataInterface() {propertyValues {Curve {keys [
           {arriveTangent=0.0007, interpMode=RCIM_Cubic, leaveTangent=0.0007, tangentMode=RCTM_User, time=0, value=1.2}
           {arriveTangent=0.2484, interpMode=RCIM_Cubic, leaveTangent=0.2484, tangentMode=RCTM_User, time=1, value=0.15}
           {arriveTangent=-0.0359, interpMode=RCIM_Cubic, leaveTangent=-0.0359, tangentMode=RCTM_User, time=1.2, value=-0.04}
           <at 1.8 0>
          ]}}}
       CurveIndex=$Particles.Age
       Scale Curve=400
      )
     )
    )
    SetAlignment=true
    SetFacing=true
   )
   &ScaleColor(
    Curve Index=&Frac_Float(Float=&Add_Float(A=$Particles.NormalizedAge, B=&RandomRangeFloat(Minimum=0, Maximum=1)))
    Linear Color Curve=DataInterface() {propertyValues {
      AlphaCurve {keys [<at 0.3319 1>]}
      BlueCurve {keys [<at 0 1>,<at 0.1961 1>,<at 0.4189 0.1729>,<at 0.6211 0.2403>,<at 0.821 0.7989>,<at 0.9983 1>]}
      GreenCurve {keys [<at 0 0.1607>,<at 0.1961 0.146>,<at 0.4189 1>,<at 0.6211 1>,<at 0.821 1>,<at 0.9983 0.1607>]}
      RedCurve {keys [<at 0 0>,<at 0.1961 0.8333>,<at 0.4189 0.6662>,<at 0.6211 0.0904>,<at 0.821 0.1713>,<at 0.9983 0>]}
     }}
    Scale Mode=::NewEnumerator2
   )
   &FN_Light_Attributes(
    Color=$Particles.Color
    Create Color Attribute=true
    Create Radius Attribute=true
    Radius=&FloatFromCurve(
     FloatCurve=DataInterface() {propertyValues {Curve {keys [
         {arriveTangent=1.9925, interpMode=RCIM_Cubic, leaveTangent=1.9925, tangentMode=RCTM_User, time=0, value=0}
         {arriveTangent=0.0224, interpMode=RCIM_Cubic, leaveTangent=0.0224, tangentMode=RCTM_User, time=1.5, value=1}
        ]}}}
     CurveIndex=$Particles.Age
     Scale Curve=100
    )
   )
   &ScaleColor(
    Color Value To Scale=$Particles.Color
    Scale Alpha=&MakeFloatFromVector(
     Vector=&VectorFromFloat(Value=&FloatFromCurve(
       FloatCurve=DataInterface() {propertyValues {Curve {keys [
           <at 0 1>
           <at 1.1 1>
           {arriveTangent=-0.291, leaveTangent=-0.291, time=1.2, value=1.5}
           {arriveTangent=-2.2992, interpMode=RCIM_Cubic, leaveTangent=-2.2992, tangentMode=RCTM_User, time=1.3, value=1.5}
           <at 1.6 1 RCIM_Cubic>
          ]}}}
       CurveIndex=$Particles.Age
      ))
     Channel=::NewEnumerator0
    )
    Scale RGB=&VectorFromFloat(Value=&FloatFromCurve(
      FloatCurve=DataInterface() {propertyValues {Curve {keys [
          <at 0 1>
          <at 1.1 1>
          {arriveTangent=-0.291, leaveTangent=-0.291, time=1.2, value=3}
          {arriveTangent=-11.1857, interpMode=RCIM_Cubic, leaveTangent=-11.1857, tangentMode=RCTM_User, time=1.3, value=3}
          <at 1.6 1 RCIM_Cubic>
         ]}}}
      CurveIndex=$Particles.Age
     ))
    ScaleA=true
    ScaleRGB=true
   ) {module=ScaleColor001}
   &GenerateLocationEvent(
    Event Generation Enabled:NiagaraBool=&Boolean_AndOperation(
     Bool1=&SetBoolByFloatComparison(A=$Particles.Age, B=1.5)
     Bool2=&SetBoolByFloatComparison(A=&VectorLength(Vector=$Particles.SkeletalMeshLocation.SampledVelocity), B=100)
    )
    Event Send Rate=25
    Use Event Probability=false
    UseEventDelay=false
   )
  ]
  rendererClasses [&NiagaraSpriteRendererProperties, &NiagaraLightRendererProperties]
  renderers [
   {rendererClass=&NiagaraSpriteRendererProperties}
   {rendererClass=&NiagaraLightRendererProperties, rendererIndex=1}
  ]
 }
 Disks {
  emitterUpdateScript [
   &EmitterState(
    Life Cycle Mode=::NewEnumerator1
    Loop Behavior=::NewEnumerator1
    Loop Delay=$User.LoopDelay
    Loop Duration Mode=::NewEnumerator1
    UseLoopDelay=true
   )
   &SpawnBurst_Instantaneous(Spawn Count=6, Use Spawn Probability=false, UseLoopCountLimit=false)
  ]
  particleSpawnScript [
   &PartitionParticles(Particles Per Partition=3, Partition Mode:ENiagara_PartitionMode=::NewEnumerator1)
   &FN_SetSpawnGroup() {enabled=false}
   &InitializeParticle(
    Color=<rgba 1 1 1 0.5>
    Lifetime=&RandomRangeFloat(Minimum=6, Maximum=10)
    Mesh Scale=&MakeVector(
     X=&SelectFloatFromArray(
      Direct Array Index=$Particles.PartitionParticles.Index In Partition
      Float Selection Array=DataInterface() {propertyValues {FloatData [0.85,0.95,1.05,1.15]}}
     )
     Y=&SelectFloatFromArray(
      Direct Array Index=$Particles.PartitionParticles.Index In Partition
      Float Selection Array=DataInterface() {propertyValues {FloatData [0.85,0.95,1.05,1.15]}}
     )
     Z=&RandomRangeFloat(Minimum=0.025, Maximum=0.05)
    )
    Mesh Scale Mode=::NewEnumerator4
    Position Offset:NiagaraPosition=$User.Position Offset
    UsePositionOffset=true
    Write Material Random=true
   )
   &AddVelocity(Velocity=<0 0 20>) {enabled=false}
   &InitialMeshOrientation(Mesh Orientation Mode=::NewEnumerator2, Rotation=&MakeVector(X=0, Y=0, Z=0), Use Rotation Vector=true)
  ]
  particleUpdateScript [
   &ParticleState(Kill Particles When Lifetime Has Elapsed=false)
   &SkeletalMeshLocation(
    Bone / Socket Index=$Particles.PartitionParticles.Partition
    Bone Sampling Mode=::NewEnumerator4
    MaskSpawnGroup=false
    Orientation Sampling=::NewEnumerator1
    Sampled Position Offset=&AddVectorToPosition(
     Position=$User.Position Offset
     Vector=&MakeVector(
      X=0
      Y=0
      Z=&FloatFromCurve(
       FloatCurve=DataInterface() {propertyValues {Curve {keys [
           {arriveTangent=9.0725, interpMode=RCIM_Cubic, time=0, value=0}
           {arriveTangent=-0.0148, interpMode=RCIM_Cubic, leaveTangent=-0.0148, tangentMode=RCTM_User, time=0.1, value=0.7}
           {arriveTangent=0.0914, interpMode=RCIM_Cubic, leaveTangent=0.0914, tangentMode=RCTM_User, time=0.2, value=-0.45}
           {arriveTangent=-0.0998, interpMode=RCIM_Cubic, leaveTangent=-0.0998, tangentMode=RCTM_User, time=0.3, value=0.25}
           <at 0.4 -0.15 RCIM_Cubic>
           {arriveTangent=0.0917, interpMode=RCIM_Cubic, leaveTangent=0.0917, tangentMode=RCTM_User, time=0.5, value=0.25}
           {arriveTangent=-0.0205, interpMode=RCIM_Cubic, leaveTangent=-0.0205, tangentMode=RCTM_User, time=0.6, value=-0.375}
           {arriveTangent=0.6667, interpMode=RCIM_Cubic, leaveTangent=0.6667, time=0.7, value=0.525}
           {arriveTangent=0.0938, interpMode=RCIM_Cubic, leaveTangent=0.0938, tangentMode=RCTM_User, time=0.9, value=-0.175}
           {interpMode=RCIM_Cubic, leaveTangent=-0.0405, time=1, value=0}
          ]}}}
       CurveIndex=&Frac_Float(Float=$Particles.NormalizedAge)
       Scale Curve=&SelectFloatFromArray(
        Direct Array Index=$Particles.PartitionParticles.Index In Partition
        Float Selection Array=DataInterface() {propertyValues {FloatData [4,1.5,-2]}}
       )
      )
     )
    )
    Skeletal Mesh=$User.Skeletal Mesh
   )
   &UpdateMeshOrientation(
    Facing Direction=&RotateVectorByQuaternion(
     Vector To Rotate=<1 0 0>
     Quaternion=&MakeQuatFromAxisAngle(
      Axis=&RandomRangeVector(Minimum=<1 0 0>, Maximum=<1 0 0>)
      AngleInDegrees=&FloatFromCurve(
       FloatCurve=DataInterface() {propertyValues {Curve {keys [{arriveTangent=-1.9238, interpMode=RCIM_Cubic, leaveTangent=-1.9238, tangentMode=RCTM_User, time=0, value=1.2}, <at 1.8 0 RCIM_Cubic>]}}}
       CurveIndex=$Particles.Age
       Scale Curve=&SelectFloatFromArray(
        Array Sampling Mode=::NewEnumerator0
        Direct Array Index=?
        Float Selection Array=DataInterface() {propertyValues {FloatData [300,-250]}}
        Random Seed=0
       )
      )
      NormalizeAxis=true
     )
    )
    Facing Interpolation=false
    Orientation Method=::NewEnumerator1
    Override Quat=false
    TransformFacing=false
    TransformUp=false
   ) {module=UpdateMeshOrientation001}
   &UpdateMeshOrientation(
    Rotation Coordinate Space=::NewEnumerator0
    Rotation Rate=&FloatFromCurve(
     FloatCurve=DataInterface() {propertyValues {Curve {keys [
         {interpMode=RCIM_Cubic, tangentMode=RCTM_User, time=0.05, value=0}
         {interpMode=RCIM_Cubic, tangentMode=RCTM_User, time=0.3, value=0.2}
         {arriveTangent=-0.0568, interpMode=RCIM_Cubic, leaveTangent=-0.0568, tangentMode=RCTM_User, time=0.5, value=0.05}
         {arriveTangent=-0.1, interpMode=RCIM_Cubic, leaveTangent=-0.1, time=0.8, value=0.2}
         <at 1 0 RCIM_Cubic>
        ]}}}
     CurveIndex=&Frac_Float(Float=$Particles.NormalizedAge)
     Scale Curve=&RandomRangeFloat(Minimum=-2, Maximum=2)
    )
    Rotation Vector=<0 0 1>
   )
   &ScaleMeshSize(Scale Factor=&Multiply_VectorByFloat(
     Vector=&MakeVector(
      X=&FloatFromCurve(
       FloatCurve=DataInterface() {propertyValues {Curve {keys [{interpMode=RCIM_Cubic, tangentMode=RCTM_User, time=0, value=0.3}, {interpMode=RCIM_Cubic, tangentMode=RCTM_User, time=1, value=1}]}}}
       CurveIndex=$Particles.Age
      )
      Y=&FloatFromCurve(
       FloatCurve=DataInterface() {propertyValues {Curve {keys [{interpMode=RCIM_Cubic, tangentMode=RCTM_User, time=0, value=0}, {interpMode=RCIM_Cubic, tangentMode=RCTM_User, time=1, value=1}]}}}
       CurveIndex=$Particles.Age
      )
      Z=&FloatFromCurve(
       FloatCurve=DataInterface() {propertyValues {Curve {keys [<at 0.49 1>,<at 0.5 0.8>,<at 0.6 0.8>,<at 0.61 1>]}}}
       CurveIndex=&Frac_Float(Float=$Particles.NormalizedAge)
      )
     )
     Float=$User.EffectScale
    ))
   &ScaleMeshSize(
    Initial Mesh Scale:Vector3f=$Particles.Scale
    Scale Factor=&VectorFromFloat(Value=&FloatFromCurve(
      FloatCurve=DataInterface() {propertyValues {
        bOverrideOptimizeThreshold=true
        Curve {keys [{interpMode=RCIM_Cubic, tangentMode=RCTM_User, time=0, value=2}, {interpMode=RCIM_Cubic, tangentMode=RCTM_User, time=1, value=2}]}
       }}
      CurveIndex=$Particles.Age
     ))
   ) {enabled=false, module=ScaleMeshSize001}
   &SkeletalMeshLocation(
    Bone / Socket Index=1
    Bone Sampling Mode=::NewEnumerator4
    MaskSpawnGroup=false
    Orientation Sampling=::NewEnumerator1
    Position Sampling=::NewEnumerator1
    Skeletal Mesh=$User.Skeletal Mesh
    Write To Particle Position=false
   ) {enabled=false, module=SkeletalMeshLocation001}
   &SolveForcesAndVelocity(Clamp Velocity=false, Limit Acceleration=false)
   &DynamicMaterialParameters(
    Index 0 Param 1=&FloatFromCurve(
     FloatCurve=DataInterface() {propertyValues {Curve {keys [
         {arriveTangent=12.0246, interpMode=RCIM_Cubic, leaveTangent=12.0246, tangentMode=RCTM_User, time=0, value=-0.1}
         {arriveTangent=1.4, interpMode=RCIM_Cubic, leaveTangent=1.4, time=0.1, value=0.85}
         {arriveTangent=-0.0833, interpMode=RCIM_Cubic, leaveTangent=-0.0833, time=0.5, value=0.6}
         {arriveTangent=-1.4, interpMode=RCIM_Cubic, leaveTangent=-1.4, time=0.7, value=0.8}
         <at 1 -0.1 RCIM_Cubic>
        ]}}}
     CurveIndex=&Frac_Float(Float=$Particles.NormalizedAge)
     Scale Curve=&RandomRangeFloat(Minimum=0.5, Maximum=1)
    )
    Module.Param0WriteEnabled=true
    Module.Param1WriteEnabled=true
    Module.Param2WriteEnabled=true
    Module.Param3WriteEnabled=true
   )
   &ScaleColor(
    Curve Index=&Frac_Float(Float=&Add_Float(A=$Particles.NormalizedAge, B=&RandomRangeFloat(Minimum=0, Maximum=1)))
    Linear Color Curve=DataInterface() {propertyValues {
      AlphaCurve {keys [<at 0.3319 1>]}
      BlueCurve {keys [<at 0 1>,<at 0.1961 1>,<at 0.4189 0.287>,<at 0.6211 0.4446>,<at 0.821 0.7989>,<at 0.9983 1>]}
      GreenCurve {keys [<at 0 0.4755>,<at 0.1961 0.235>,<at 0.4189 1>,<at 0.6211 1>,<at 0.821 1>,<at 0.9983 0.4755>]}
      RedCurve {keys [<at 0 0.375>,<at 0.1961 0.8507>,<at 0.4189 0.7122>,<at 0.6211 0.335>,<at 0.821 0.1713>,<at 0.9983 0.375>]}
     }}
    Scale Mode=::NewEnumerator2
   ) {module=ScaleColor002}
   &ScaleColor(
    Color Value To Scale=$Particles.Color
    Scale Alpha=0.8
    Scale RGB=&VectorFromFloat(Value=&FloatFromCurve(
      FloatCurve=DataInterface() {propertyValues {Curve {keys [
          {arriveTangent=-0.0272, interpMode=RCIM_Cubic, leaveTangent=-0.0272, tangentMode=RCTM_User, time=0, value=1}
          {interpMode=RCIM_Cubic, tangentMode=RCTM_User, time=0.2, value=1.375}
          {arriveTangent=0.0222, interpMode=RCIM_Cubic, leaveTangent=0.0222, tangentMode=RCTM_User, time=0.5, value=1}
          <at 0.8 1.125 RCIM_Cubic>
          <at 1 1 RCIM_Cubic>
         ]}}}
      CurveIndex=&Frac_Float(Float=$Particles.NormalizedAge)
     ))
    ScaleA=true
    ScaleRGB=true
   )
   &ScaleColor(
    Color Value To Scale=$Particles.Color
    Scale RGB=&VectorFromFloat(Value=&FloatFromCurve(
      FloatCurve=DataInterface() {propertyValues {Curve {keys [
          <at 0 1>
          <at 1.1 1>
          {arriveTangent=-0.291, leaveTangent=-0.291, time=1.2, value=2.5}
          {arriveTangent=-9.4361, interpMode=RCIM_Cubic, leaveTangent=-9.4361, tangentMode=RCTM_User, time=1.3, value=2.5}
          <at 1.6 1 RCIM_Cubic>
         ]}}}
      CurveIndex=$Particles.Age
     ))
    ScaleA=true
    ScaleRGB=true
   ) {module=ScaleColor001}
   &FN_Mesh_Attributes(
    Color=&MultiplyLinearColor(
     Linear Color=$Particles.Color
     Scale Factor=&MakeFloat4FromVectorAndFloat("Vector (RGB)"=&VectorFromFloat(Value=1), "Float (Alpha)"=0.2)
    )
    Create Color Attribute=true
    Create Scale Attribute=true
    Scale=&Multiply_Vector(A=$Particles.Scale, B=<1 1 1.5>)
    Write to Module Namespace=true
   )
  ]
  rendererClasses [&NiagaraMeshRendererProperties]
  renderers [
   {rendererClass=&NiagaraMeshRendererProperties, rendererIndex=0}
   {rendererClass=&NiagaraMeshRendererProperties, rendererIndex=1}
  ]
 }
 FN_DebugGnomon {
  bEnabled=false
  emitterUpdateScript [
   &EmitterState(
    Life Cycle Mode=::NewEnumerator1
    Loop Behavior=::NewEnumerator1
    Loop Duration Mode=::NewEnumerator1
    Scalability Mode=::NewEnumerator1
    UseLoopDelay=false
   )
   &SpawnBurst_Instantaneous(Spawn Count=2, Use Spawn Probability=false, UseLoopCountLimit=false)
  ]
  particleSpawnScript [&InitializeParticle(
    Color Mode=::NewEnumerator0
    Lifetime=5
    Mesh Scale Mode=::NewEnumerator3
    Mesh Uniform Scale=5
    UsePositionOffset=false
    Write Material Random=true
   )]
  particleUpdateScript [
   &ParticleState(Kill Particles When Lifetime Has Elapsed=false)
   &SkeletalMeshLocation(
    Bone / Socket Index=&ReturnExecIndex()
    Bone Sampling Mode=::NewEnumerator4
    MaskSpawnGroup=false
    Skeletal Mesh=$User.Skeletal Mesh
    Use Sampled Mesh Attribute as Ribbon ID=false
   )
  ]
  rendererClasses [&NiagaraMeshRendererProperties]
  renderers [{rendererClass=&NiagaraMeshRendererProperties, rendererIndex=0}]
 }
 Flash_Flare {
  emitterUpdateScript [
   &EmitterState(
    Life Cycle Mode=::NewEnumerator1
    Loop Behavior=::NewEnumerator1
    Loop Delay=$User.LoopDelay
    Loop Duration Mode=::NewEnumerator1
    UseLoopDelay=true
   )
   &SpawnBurst_Instantaneous(Spawn Count=1, Use Spawn Probability=true)
  ]
  particleSpawnScript [
   &FN_SetSpawnGroup(Number of Spawn Groups=2)
   &InitializeParticle(
    Color=&SelectLinearColorFromArray(
     Array Sampling Mode=::NewEnumerator0
     Direct Array Index=?
     Color Selection Array=DataInterface() {propertyValues {ColorData [<rgba 10 0.57 6.6844 1>, <rgba 0.76 7.6709 10 1>, <rgba 1.14 10 3.4678 1>, <rgba 2.8622 1.06 10 1>]}}
     Random Seed=0
    )
    Lifetime=0.4
    Sprite Size Mode=::NewEnumerator3
    Uniform Sprite Size=120
    UsePositionOffset=false
    Write Material Random=true
   )
   &SkeletalMeshLocation(
    Bone Sampling Mode=::NewEnumerator4
    Interpolate Mesh Sampling=false
    MaskSpawnGroup=false
    Orientation Sampling=::NewEnumerator1
    Skeletal Mesh=$User.Skeletal Mesh
   )
   &DynamicMaterialParameters(
    Index 0 Param 1=0.5
    Module.Param0WriteEnabled=true
    Module.Param1WriteEnabled=true
    Module.Param2WriteEnabled=true
    Module.Param3WriteEnabled=true
   )
  ]
  particleUpdateScript [
   &ParticleState(Kill Particles When Lifetime Has Elapsed=false)
   &ScaleSpriteSize(Uniform Curve Sprite Scale=DataInterface() {propertyValues {Curve {keys [
        {arriveTangent=9.0725, interpMode=RCIM_Cubic, leaveTangent=9.0725, tangentMode=RCTM_User, time=0, value=0.6}
        {arriveTangent=-1.14, interpMode=RCIM_Cubic, leaveTangent=-1.14, tangentMode=RCTM_User, time=0.1408, value=0.9847}
        {arriveTangent=-0.6284, interpMode=RCIM_Cubic, leaveTangent=-0.6284, tangentMode=RCTM_User, time=1, value=0}
       ]}}})
   &SpriteRotationRate(Rotation Rate=360) {enabled=false}
   &ScaleColor(
    Scale Alpha=&FloatFromCurve(
     FloatCurve=DataInterface() {propertyValues {Curve {keys [
         {arriveTangent=9.0725, interpMode=RCIM_Cubic, leaveTangent=9.0725, tangentMode=RCTM_User, time=0, value=0}
         {arriveTangent=-2.5249, interpMode=RCIM_Cubic, leaveTangent=-2.5249, tangentMode=RCTM_User, time=0.1408, value=0.9847}
         {arriveTangent=-0.0405, interpMode=RCIM_Cubic, leaveTangent=-0.0405, tangentMode=RCTM_User, time=1, value=0}
        ]}}}
     Scale Curve=0.1
    )
    Scale RGB=&VectorFromFloat(Value=&FloatFromCurve(
      FloatCurve=DataInterface() {propertyValues {Curve {keys [
          {arriveTangent=9.0725, interpMode=RCIM_Cubic, leaveTangent=9.0725, tangentMode=RCTM_User, time=0, value=0.75}
          {arriveTangent=-2.5249, interpMode=RCIM_Cubic, leaveTangent=-2.5249, tangentMode=RCTM_User, time=0.1408, value=0.9847}
          {arriveTangent=-0.0405, interpMode=RCIM_Cubic, leaveTangent=-0.0405, tangentMode=RCTM_User, time=1, value=0}
         ]}}}
      Scale Curve=5
     ))
    ScaleA=true
    ScaleRGB=true
   )
   &CameraOffset(Camera Offset Amount=5) {module=CameraOffset001}
  ]
  rendererClasses [&NiagaraSpriteRendererProperties]
  renderers [{rendererClass=&NiagaraSpriteRendererProperties}]
 }
 GlitchySquare {
  emitterUpdateScript [
   &EmitterState(
    Life Cycle Mode=::NewEnumerator1
    Loop Behavior=::NewEnumerator1
    Loop Delay=$User.LoopDelay
    Loop Duration Mode=::NewEnumerator1
    UseLoopDelay=true
   )
   &SpawnRate(SpawnRate=25, Use Spawn Probability=false)
   &SpawnBurst_Instantaneous(
    Spawn Count=&RandomRangeInt(Minimum=35, Maximum=50)
    Use Spawn Probability=false
    UseLoopCountLimit=false
   ) {enabled=false}
  ]
  particleSpawnScript [
   &FN_SetSpawnGroup(Number of Spawn Groups=2)
   &InitializeParticle(
    Color=&MakeLinearColorFromVectorAndFloat(
     "Vector (RGB)"=&VectorFromFloat(Value=&RandomRangeFloat(Minimum=3, Maximum=6))
     "Float (Alpha)"=1
    )
    Lifetime Max=0.5
    Lifetime Min=0.25
    Lifetime Mode=::NewEnumerator1
    Position Offset:NiagaraPosition=$User.Position Offset
    Sprite Size=&Multiply_Vector2D_ByFloat(Float=$User.EffectScale, Vector2D=&RandomRangeVector2D(Minimum=<4 4>, Maximum=<6 6>))
    Sprite Size Mode=::NewEnumerator4
    Sprite UV Mode=::NewEnumerator3
    UsePositionOffset=true
    Write Material Random=true
   )
   &SkeletalMeshLocation(
    Bone / Socket Index=$Particles.SpawnGroup
    Bone Sampling Mode=::NewEnumerator4
    MaskSpawnGroup=false
    Orientation Sampling=::NewEnumerator1
    Sampled Position Offset:NiagaraPosition=$User.Position Offset
    Skeletal Mesh=$User.Skeletal Mesh
   )
   &ShapeLocation(
    Disc Coverage=1
    MaskSpawnGroup=false
    Override Seed=false
    Ring Radius=&Multiply_Float(A=100, B=$User.EffectScale)
    Shape Primitive=::NewEnumerator4
    UseOwnerScale=true
   ) {enabled=false}
   &AddVelocity(Velocity=&MakeVector(
     X=0
     Y=0
     Z=&Multiply_Float(A=&RandomRangeFloat(Minimum=-100, Maximum=500), B=$User.EffectScale)
    )) {enabled=false}
   &ShapeLocation(MaskSpawnGroup=false, Override Seed=false, Sphere Radius=20, UseOwnerScale=true) {module=ShapeLocation001}
   &AddVelocity(
    Module.Use Falloff Distance=false
    Override Seed=false
    Velocity Mode=::NewEnumerator1
    Velocity Speed=&RandomRangeFloat(Minimum=30, Maximum=70)
   ) {module=AddVelocity001}
  ]
  particleUpdateScript [
   &ParticleState()
   &GravityForce() {enabled=false}
   &Drag(Drag=&RandomRangeFloat(Minimum=2, Maximum=5), Use Linear Drag=true, Use Rotational Drag=false)
   &ScaleSpriteSize(Scale Factor=&Vector2DFromCurve(
     Vector2Curve=DataInterface() {propertyValues {
       XCurve {keys [<at 0 1>,<at 1 1>]}
       YCurve {keys [
         {arriveTangent=9.0725, interpMode=RCIM_Cubic, leaveTangent=9.0725, tangentMode=RCTM_User, time=0, value=0.6}
         {arriveTangent=-2.5249, interpMode=RCIM_Cubic, leaveTangent=-2.5249, tangentMode=RCTM_User, time=0.6, value=0.75}
         {arriveTangent=-0.0405, interpMode=RCIM_Cubic, leaveTangent=-0.0405, tangentMode=RCTM_User, time=1, value=0.2}
        ]}
      }}
     ScaleCurve=<1 2>
    )) {enabled=false}
   &ScaleSpriteSize(Scale Factor=&Vector2DFromFloat(Value=&FloatFromCurve(FloatCurve=DataInterface() {propertyValues {Curve {keys [
          {arriveTangent=9.0725, interpMode=RCIM_Cubic, leaveTangent=9.0725, tangentMode=RCTM_User, time=0, value=1}
          {arriveTangent=-2.5249, interpMode=RCIM_Cubic, leaveTangent=-2.5249, tangentMode=RCTM_User, time=0.2, value=1.375}
          {arriveTangent=-0.2083, interpMode=RCIM_Cubic, leaveTangent=-0.2083, time=0.4, value=1}
          {arriveTangent=-1.6667, interpMode=RCIM_Cubic, leaveTangent=-1.6667, time=0.8, value=1.25}
          {arriveTangent=-7.0578, interpMode=RCIM_Cubic, leaveTangent=-7.0578, tangentMode=RCTM_User, time=1, value=0}
         ]}}}))) {module=ScaleSpriteSize001}
   &ScaleColor(
    Scale Mode=::NewEnumerator1
    Scale RGBA=&Vector4FromCurve(
     CurveIndex=&Frac_Float(Float=&Add_Float(A=$Particles.NormalizedAge, B=&RandomRangeFloat(Minimum=0, Maximum=1)))
     Vector4Curve=DataInterface() {propertyValues {
       AlphaCurve {keys [<at 0.0012 0>,<at 0.3319 0.25>,<at 1 0>]}
       BlueCurve {keys [<at 0 1>,<at 0.2196 1>,<at 0.4293 0>,<at 0.7138 0.2403>,<at 1 0.7989>]}
       GreenCurve {keys [<at 0 0.1607>,<at 0.2196 0>,<at 0.4293 0.4093>,<at 0.7138 1>,<at 1 1>]}
       RedCurve {keys [<at 0 0>,<at 0.2196 0.8048>,<at 0.4293 1>,<at 0.7138 0.0904>,<at 1 0.1713>]}
      }}
    )
    ScaleRGBA=true
   )
   &InheritSourceMovement(
    Applied Position Scale=&VectorFromFloat(Value=&FloatFromCurve(FloatCurve=DataInterface() {propertyValues {Curve {keys [
          {arriveTangent=-2.5859, interpMode=RCIM_Cubic, leaveTangent=-2.5859, tangentMode=RCTM_User, time=0, value=0.65}
          {interpMode=RCIM_Cubic, tangentMode=RCTM_User, time=0.8, value=0}
         ]}}}))
    ApplyForce=false
    ApplyPosition=true
    ApplyPositionFalloff=false
   )
   &SolveForcesAndVelocity(Clamp Velocity=false, Limit Acceleration=false)
   &SpriteFacingAndAlignment(Sprite Facing=$Particles.ShapeLocation.ShapeTangent, WriteAlignment=false, WriteFacing=true)
   &FN_Sprite_Attributes(
    Color=&MultiplyLinearColor(Linear Color=$Particles.Color, Scale Factor=<1 1 1 1>)
    Create Color Attribute=true
    Create SpriteSize Attribute=true
    SpriteSize=&Multiply_Vector2D_ByFloat(Float=1.5, Vector2D=$Particles.SpriteSize)
    Write to Module Namespace=true
   )
  ]
  rendererClasses [&NiagaraSpriteRendererProperties]
  renderers [
   {rendererClass=&NiagaraSpriteRendererProperties}
   {rendererClass=&NiagaraSpriteRendererProperties, rendererIndex=1}
  ]
 }
 GlitchySquareBurst {
  emitterUpdateScript [
   &EmitterState(
    Life Cycle Mode=::NewEnumerator1
    Loop Behavior=::NewEnumerator1
    Loop Delay=$User.LoopDelay
    Loop Duration Mode=::NewEnumerator1
    UseLoopDelay=true
   )
   &SpawnRate(SpawnRate=10, Use Spawn Probability=false) {enabled=false}
   &SpawnBurst_Instantaneous(
    Spawn Count=&RandomRangeInt(Minimum=35, Maximum=50)
    Use Spawn Probability=false
    UseLoopCountLimit=false
   )
  ]
  particleSpawnScript [
   &FN_SetSpawnGroup(Number of Spawn Groups=2)
   &InitializeParticle(
    Color=&MakeLinearColorFromVectorAndFloat(
     "Vector (RGB)"=&VectorFromFloat(Value=&RandomRangeFloat(Minimum=3, Maximum=6))
     "Float (Alpha)"=1
    )
    Lifetime Max=0.75
    Lifetime Min=0.2
    Lifetime Mode=::NewEnumerator1
    Sprite Size Mode=::NewEnumerator3
    Uniform Sprite Size=&Multiply_Float(A=&RandomRangeFloat(Minimum=4, Maximum=7), B=$User.EffectScale)
    UsePositionOffset=false
    Write Material Random=true
   )
   &SkeletalMeshLocation(
    Bone / Socket Index=$Particles.SpawnGroup
    Bone Sampling Mode=::NewEnumerator4
    MaskSpawnGroup=false
    Orientation Sampling=::NewEnumerator1
    Skeletal Mesh=$User.Skeletal Mesh
   )
   &ShapeLocation(MaskSpawnGroup=false, Override Seed=false, Sphere Radius=15, UseOwnerScale=true)
   &AddVelocity(
    Rotation Coordinate Space=::Simulation
    Velocity=&NormalizeVector(Vector To Normalize=$Particles.ShapeLocation.ShapeVector, Safe Normalize=true, Safe Fallback Threshold=0, Safe Fallback Vector=<0 0 0.001>)
    Velocity Speed Scale=&RandomRangeFloat(Minimum=75, Maximum=100)
   )
  ]
  particleUpdateScript [
   &ParticleState()
   &GravityForce() {enabled=false}
   &Drag(Drag=&RandomRangeFloat(Minimum=3, Maximum=5), Use Linear Drag=true, Use Rotational Drag=false)
   &ScaleSpriteSize(Scale Factor=&Vector2DFromFloat(Value=&FloatFromCurve(FloatCurve=DataInterface() {propertyValues {Curve {keys [
          {arriveTangent=9.0725, interpMode=RCIM_Cubic, leaveTangent=9.0725, tangentMode=RCTM_User, time=0, value=1}
          {arriveTangent=0.059, interpMode=RCIM_Cubic, leaveTangent=0.059, tangentMode=RCTM_User, time=0.2, value=1.875}
          {arriveTangent=-1.694, interpMode=RCIM_Cubic, leaveTangent=-1.694, tangentMode=RCTM_User, time=1, value=0}
         ]}}})))
   &ScaleColor(
    Scale Mode=::NewEnumerator1
    Scale RGBA=&Vector4FromCurve(
     CurveIndex=&Frac_Float(Float=&Add_Float(A=$Particles.NormalizedAge, B=&RandomRangeFloat(Minimum=0, Maximum=1)))
     Vector4Curve=DataInterface() {propertyValues {
       AlphaCurve {keys [<at 0.0012 0>,<at 0.3319 0.25>,<at 1 0>]}
       BlueCurve {keys [<at 0 1>,<at 0.2196 1>,<at 0.4293 0>,<at 0.7138 0.2403>,<at 1 0.7989>]}
       GreenCurve {keys [<at 0 0.1607>,<at 0.2196 0>,<at 0.4293 0.4093>,<at 0.7138 1>,<at 1 1>]}
       RedCurve {keys [<at 0 0>,<at 0.2196 0.8048>,<at 0.4293 1>,<at 0.7138 0.0904>,<at 1 0.1713>]}
      }}
    )
    ScaleRGBA=true
   )
   &SolveForcesAndVelocity(Clamp Velocity=false, Limit Acceleration=false)
   &SpriteRotationRate(Rotation Rate=&RandomRangeFloat(Minimum=-300, Maximum=300))
   &SpriteFacingAndAlignment(Sprite Facing=$Particles.ShapeLocation.ShapeTangent, WriteAlignment=false, WriteFacing=true) {enabled=false}
   &FN_Sprite_Attributes(
    Color=&MultiplyLinearColor(Linear Color=$Particles.Color, Scale Factor=<1 1 1 1>)
    Create Color Attribute=true
    Create SpriteSize Attribute=true
    SpriteSize=&Multiply_Vector2D_ByFloat(Float=1.5, Vector2D=$Particles.SpriteSize)
    Write to Module Namespace=true
   )
  ]
  rendererClasses [&NiagaraSpriteRendererProperties]
  renderers [
   {rendererClass=&NiagaraSpriteRendererProperties}
   {rendererClass=&NiagaraSpriteRendererProperties, rendererIndex=1}
  ]
 }
 Notes_Line {
  emitterSpawnScript [/SetVariables(Emitter.Rand:NiagaraFloat=&RandomRangeFloat(Minimum=0, Maximum=1)) {isSetParameters=true}]
  emitterUpdateScript [
   &EmitterState(
    Life Cycle Mode=::NewEnumerator1
    Loop Behavior=::NewEnumerator1
    Loop Delay=$User.LoopDelay
    Loop Duration Mode=::NewEnumerator1
    UseLoopDelay=true
   )
   &SpawnRate(SpawnRate=100, Use Spawn Probability=false) {enabled=false}
   &SpawnBurst_Instantaneous(Spawn Count=120, Spawn Time=0.05, Use Spawn Probability=false, UseLoopCountLimit=false)
  ]
  particleSpawnScript [
   &PartitionParticles(
    Normalized Index In Partition Includes One=true
    Number Of Partitions=3
    Particles Per Partition=40
    Partition Mode:ENiagara_PartitionMode=::NewEnumerator1
    Write Partition to Ribbon ID=::NewEnumerator0
   )
   &InitializeParticle(
    Color Mode=::NewEnumerator0
    Lifetime=0.4
    Ribbon Width=1
    Ribbon Width Mode=::NewEnumerator1
    UsePositionOffset=false
    Write Material Random=true
   )
   &SkeletalMeshLocation(
    Bone Sampling Mode=::NewEnumerator4
    Interpolate Mesh Sampling=false
    MaskSpawnGroup=false
    Skeletal Mesh=$User.Skeletal Mesh
    Use Sampled Mesh Attribute as Ribbon ID=false
   )
   &AddVelocity() {enabled=false}
   /SetVariables(Particles.RibbonLinkOrder:NiagaraFloat=$Particles.PartitionParticles.Normalized Index In Partition) {isSetParameters=true}
  ]
  particleUpdateScript [
   &ParticleState()
   &Color(Color=&SelectLinearColorFromArray(
     Array Sampling Mode=::NewEnumerator1
     Direct Array Index=$Particles.PartitionParticles.Partition
     Color Selection Array=DataInterface() {propertyValues {ColorData [<rgba 0 1.618 2 1>, <rgba 1.8748 0.4343 2 1>, <rgba 3 1.5719 0.2969 1>]}}
     Random Seed=?
    ))
   &ScaleColor(
    Color Value To Scale=$Particles.Color
    Scale Alpha=&FloatFromCurve(FloatCurve=DataInterface() {propertyValues {Curve {keys [
         {arriveTangent=-2.3502, interpMode=RCIM_Cubic, leaveTangent=-2.3502, tangentMode=RCTM_User, time=0, value=1}
         {interpMode=RCIM_Cubic, tangentMode=RCTM_User, time=1, value=0}
        ]}}})
    ScaleA=true
    ScaleRGB=true
   )
   &SolveForcesAndVelocity(Clamp Velocity=false, Limit Acceleration=false)
   &RotateAroundPoint(
    Radius:NiagaraFloat=&FloatFromCurve(
     FloatCurve=DataInterface() {propertyValues {Curve {keys [
         {arriveTangent=2.1122, interpMode=RCIM_Cubic, leaveTangent=2.1122, tangentMode=RCTM_User, time=0, value=0}
         {arriveTangent=0.0922, interpMode=RCIM_Cubic, leaveTangent=0.0922, tangentMode=RCTM_User, time=1, value=1}
        ]}}}
     Scale Curve=&SelectFloatFromArray(Direct Array Index=$Particles.PartitionParticles.Partition, Float Selection Array=DataInterface() {propertyValues {FloatData [20,22.5,25]}})
    )
    Rotation Center:NiagaraPosition=$Particles.Initial.Position
    Rotation Phase:NiagaraFloat=&Add_Float(A=$Emitter.Rand, B=$Particles.PartitionParticles.Normalized Index In Partition)
    X Axis:Vector3f=&ReturnCameraVector(Camera Vectors=::NewEnumerator2, Coordinate Space=::Simulation)
    Y Axis:Vector3f=&ReturnCameraVector(Camera Vectors=::NewEnumerator3, Coordinate Space=::Simulation)
   )
   /SetVariables(Particles.Position=&AddVectorToPosition(
     Position=$Particles.Position
     Vector=&Multiply_VectorByFloat(
      Vector=$Particles.RotateAroundPoint.VectorFromCenter
      Float=&FloatFromCurve(
       FloatCurve=DataInterface() {propertyValues {Curve {keys [
           {arriveTangent=9.0725, interpMode=RCIM_Cubic, leaveTangent=9.0725, tangentMode=RCTM_User, time=0, value=0}
           {arriveTangent=-1.8917, interpMode=RCIM_Cubic, leaveTangent=-1.8917, tangentMode=RCTM_User, time=0.6, value=0.425}
           {arriveTangent=-0.0208, interpMode=RCIM_Cubic, leaveTangent=-0.0208, tangentMode=RCTM_User, time=1, value=0}
          ]}}}
       Scale Curve=&Multiply_Float(
        A=40
        B=&FloatFromCurve(
         FloatCurve=DataInterface() {propertyValues {Curve {keys [
             <at 0 0.8>
             {arriveTangent=-6.0547, time=0.025, value=0}
             <at 0.08 0>
             <at 0.1 0.5>
             <at 0.12 0>
             {arriveTangent=5.533, leaveTangent=5.533, time=0.16, value=0}
             <at 0.202 0.75>
             {arriveTangent=-7.2386, leaveTangent=-7.2386, time=0.24, value=0}
             <at 0.28 0>
             <at 0.3 0.6>
             <at 0.32 0>
             {arriveTangent=3.8652, leaveTangent=3.8652, time=0.3523, value=0}
             <at 0.3723 0.5>
             {arriveTangent=-3.7702, leaveTangent=-3.7702, time=0.3923, value=0}
             {arriveTangent=8.0575, leaveTangent=8.0575, time=0.48, value=0}
             <at 0.5 0.8>
             {arriveTangent=-10.12, leaveTangent=-10.12, time=0.52, value=0}
             {arriveTangent=7.1912, leaveTangent=7.1912, time=0.6, value=0}
             <at 0.62 0.6>
             {arriveTangent=-8.6269, leaveTangent=-8.6269, time=0.64, value=0}
             {arriveTangent=15.9213, leaveTangent=15.9213, time=0.68, value=0}
             <at 0.7 0.5>
             {arriveTangent=-8.1956, leaveTangent=-8.1956, time=0.72, value=0}
             {arriveTangent=7.7748, leaveTangent=7.7748, time=0.84, value=0}
             {arriveTangent=-9.6894, leaveTangent=-9.6894, time=0.85, value=0.5}
             {arriveTangent=-7.785, leaveTangent=-7.785, time=0.86, value=0}
             {arriveTangent=9.6548, leaveTangent=9.6548, time=0.975, value=0}
             <at 1 0.8>
            ]}}}
         CurveIndex=$Particles.PartitionParticles.Normalized Index In Partition
         Scale Curve=0.5
        )
       )
      )
     )
    )) {isSetParameters=true}
   /SetVariables(Particles.Position=&AddVectorToPosition(
     Position=$Particles.Position
     Vector=&Multiply_VectorByFloat(
      Vector=$Particles.RotateAroundPoint.VectorFromCenter
      Float=&FloatFromCurve(
       FloatCurve=DataInterface() {propertyValues {Curve {keys [
           {arriveTangent=4.389, interpMode=RCIM_Cubic, leaveTangent=4.389, tangentMode=RCTM_User, time=0, value=0}
           {arriveTangent=-0.181, interpMode=RCIM_Cubic, leaveTangent=-0.181, tangentMode=RCTM_User, time=0.2, value=1}
           {arriveTangent=-0.0405, interpMode=RCIM_Cubic, leaveTangent=-0.0405, tangentMode=RCTM_User, time=1, value=-0.05}
          ]}}}
       Scale Curve=&Add_Float(
        A=&Sine(
         Normalized Angle=$Particles.PartitionParticles.Normalized Index In Partition
         Period=0.25
         Scale=&RandomRangeFloat(Minimum=5, Maximum=10)
         Bias=1
        )
        B=0
       )
      )
     )
    )) {enabled=false, isSetParameters=true}
  ]
  rendererClasses [&NiagaraRibbonRendererProperties]
  renderers [{rendererClass=&NiagaraRibbonRendererProperties}]
 }
 RainbowFlare {
  emitterUpdateScript [
   &EmitterState(
    Inactive Response=::NewEnumerator1
    Life Cycle Mode=::NewEnumerator1
    Loop Behavior=::NewEnumerator1
    Loop Delay=$User.LoopDelay
    Loop Duration=8.5
    UseLoopDelay=true
   )
   &SpawnBurst_Instantaneous(Spawn Count=1, Use Spawn Probability=false)
  ]
  particleSpawnScript [
   &InitializeParticle(
    Color=<rgba 1 1 1 0.5>
    Lifetime=0.5
    Sprite Size Mode=::NewEnumerator3
    Uniform Sprite Size=80
    UsePositionOffset=false
    Write Material Random=true
   ) {module=InitializeParticle001}
   &SkeletalMeshLocation(
    Bone Sampling Mode=::NewEnumerator4
    Interpolate Mesh Sampling=false
    MaskSpawnGroup=false
    Orientation Sampling=::NewEnumerator1
    Skeletal Mesh=$User.Skeletal Mesh
   )
   &CameraOffset(Camera Offset Amount=5)
  ]
  particleUpdateScript [
   &ParticleState()
   &SolveForcesAndVelocity(Clamp Velocity=false, Limit Acceleration=false)
   &ScaleColor(
    Scale Alpha=&FloatFromCurve(FloatCurve=DataInterface() {propertyValues {Curve {keys [
         {arriveTangent=6.4019, interpMode=RCIM_Cubic, leaveTangent=6.4019, tangentMode=RCTM_User, time=0, value=0}
         {arriveTangent=0.002, interpMode=RCIM_Cubic, leaveTangent=0.002, time=0.2, value=0.7}
         {arriveTangent=-0.9006, interpMode=RCIM_Cubic, leaveTangent=-0.9006, tangentMode=RCTM_User, time=1.0011, value=0.002}
        ]}}})
    ScaleA=true
    ScaleRGB=false
   )
   &ScaleSpriteSize(Scale Factor=&Vector2DFromFloat(Value=&FloatFromCurve(FloatCurve=DataInterface() {propertyValues {Curve {keys [
          {arriveTangent=1.0248, interpMode=RCIM_Cubic, leaveTangent=1.0248, tangentMode=RCTM_User, time=0, value=0.55}
          {arriveTangent=0.0224, interpMode=RCIM_Cubic, leaveTangent=0.0224, tangentMode=RCTM_User, time=1, value=1}
         ]}}})))
   &SpriteRotationRate(Rotation Rate=&UniformRangedFloat(Minimum=-10, Maximum=10)) {enabled=false}
  ]
  rendererClasses [&NiagaraSpriteRendererProperties]
  renderers [{rendererClass=&NiagaraSpriteRendererProperties}]
 }
 Trails {
  emitterUpdateScript [
   &EmitterState(Life Cycle Mode=::NewEnumerator1, Loop Behavior=::NewEnumerator1, Loop Delay=2, Loop Duration Mode=::NewEnumerator1, UseLoopDelay=true)
   &SpawnRate(SpawnRate=250, Use Spawn Probability=false) {enabled=false}
  ]
  particleSpawnScript [
   &InitializeParticle(
    Color=<rgba 0 4.5592 5 0.1>
    Lifetime=0.3
    Ribbon Facing Mode=::NewEnumerator1
    Ribbon Width=47.5
    Ribbon Width Mode=::NewEnumerator1
    UsePositionOffset=false
    Write Material Random=true
   )
   &AddVelocity(Velocity=<-500 0 0>) {enabled=false}
  ]
  particleUpdateScript [
   &ParticleState()
   &SolveForcesAndVelocity(Clamp Velocity=false, Limit Acceleration=false) {enabled=false}
   &ScaleRibbonWidth(Ribbon Width Scale=&FloatFromCurve(FloatCurve=DataInterface() {propertyValues {Curve {keys [{arriveTangent=-0.4441, time=0, value=1}, {arriveTangent=-1.1495, leaveTangent=-1.1495, time=1, value=0.6}]}}}))
   &ScaleColor(
    Scale Alpha=&FloatFromCurve(FloatCurve=DataInterface() {propertyValues {Curve {keys [<at 0 1>,<at 1 0>]}}})
    ScaleA=true
    ScaleRGB=true
   )
  ]
  rendererClasses [&NiagaraRibbonRendererProperties]
  renderers [{rendererClass=&NiagaraRibbonRendererProperties}]
 }
 Trails001 {
  emitterUpdateScript [
   &EmitterState(Life Cycle Mode=::NewEnumerator1, Loop Behavior=::NewEnumerator1, Loop Delay=2, Loop Duration Mode=::NewEnumerator1, UseLoopDelay=true)
   &SpawnRate(SpawnRate=250, Use Spawn Probability=false) {enabled=false}
  ]
  particleSpawnScript [
   &InitializeParticle(
    Color=<rgba 0 4.5592 5 0.1>
    Lifetime=0.3
    Ribbon Facing Mode=::NewEnumerator1
    Ribbon Width=3
    Ribbon Width Mode=::NewEnumerator1
    UsePositionOffset=false
    Write Material Random=true
   )
   &AddVelocity(Velocity=<-500 0 0>) {enabled=false}
  ]
  particleUpdateScript [
   &ParticleState()
   &SolveForcesAndVelocity(Clamp Velocity=false, Limit Acceleration=false)
   &ScaleRibbonWidth(Ribbon Width Scale=&FloatFromCurve(FloatCurve=DataInterface() {propertyValues {Curve {keys [{arriveTangent=-0.4441, time=0, value=1}, {arriveTangent=-1.1495, leaveTangent=-1.1495, time=1, value=0}]}}}))
  ]
  rendererClasses [&NiagaraRibbonRendererProperties]
  renderers [{rendererClass=&NiagaraRibbonRendererProperties}]
 }
 Volume_Bar {
  emitterUpdateScript [
   &EmitterState(
    Inactive Response=::NewEnumerator1
    Life Cycle Mode=::NewEnumerator1
    Loop Behavior=::NewEnumerator1
    Loop Delay=$User.LoopDelay
    Loop Duration Mode=::NewEnumerator1
    UseLoopDelay=true
   )
   &SpawnBurst_Instantaneous(Spawn Count=2, Spawn Time=1.2, Use Spawn Probability=false, UseLoopCountLimit=false)
   /SetVariables(Emitter.IncomingSpectrumData:LinearColor=&LerpLinearColors(
     StartColor=&MakeLinearColor(
      R:NiagaraFloat=&Sine(Normalized Angle=$Emitter.Age, Period=4, Scale=0.5, Bias=0)
      G:NiagaraFloat=&Cosine(Normalized Angle=$Emitter.Age, Period=5, Scale=1, Bias=-1)
      B:NiagaraFloat=&Sine(Normalized Angle=$Emitter.Age, Period=3.25, Scale=0.5, Bias=0.2)
      A:NiagaraFloat=&Cosine(Normalized Angle=$Emitter.Age, Period=10, Scale=1, Bias=-0.5)
     )
     EndColor=/Niagara/DynamicInputs/LinearColor/MakeLinearColorFromFloat4(Vector:Vector4f=&MakeVector4(
       X=&OneMinusFloat(Float=&Power(
         a=&OneMinusFloat(Float=&MakeFloatFromLinearColor(LinearColor=$NPC.FortniteNPC.EventAudio_Spectrum, Channel:ENiagaraLinearColor_Channels=::NewEnumerator0))
         b=0.5
        ))
       Y=&OneMinusFloat(Float=&Power(
         a=&OneMinusFloat(Float=&MakeFloatFromLinearColor(LinearColor=$NPC.FortniteNPC.EventAudio_Spectrum, Channel:ENiagaraLinearColor_Channels=::NewEnumerator1))
         b=1
        ))
       Z=&OneMinusFloat(Float=&Power(
         a=&OneMinusFloat(Float=&MakeFloatFromLinearColor(LinearColor=$NPC.FortniteNPC.EventAudio_Spectrum, Channel:ENiagaraLinearColor_Channels=::NewEnumerator2))
         b=0.75
        ))
       W=&OneMinusFloat(Float=&Power(
         a=&OneMinusFloat(Float=&MakeFloatFromLinearColor(LinearColor=$NPC.FortniteNPC.EventAudio_Spectrum, Channel:ENiagaraLinearColor_Channels=::NewEnumerator3))
         b=1
        ))
      ))
     LerpFactor=$System.IncomingAudioGate
    )) {isSetParameters=true}
  ]
  particleSpawnScript [
   &FN_SetSpawnGroup(Number of Spawn Groups=2) {enabled=false}
   &PartitionParticles(Particles Per Partition=1, Partition Mode:ENiagara_PartitionMode=::NewEnumerator1)
   &InitializeParticle(
    Color=&MakeLinearColorFromVectorAndFloat(
     "Vector (RGB)"=&VectorFromFloat(Value=&SelectFloatFromArray(
       Direct Array Index=$Particles.PartitionParticles.Index In Partition
       Float Selection Array=DataInterface() {propertyValues {FloatData [1,1]}}
      ))
     "Float (Alpha)"=&SelectFloatFromArray(
      Direct Array Index=$Particles.PartitionParticles.Index In Partition
      Float Selection Array=DataInterface() {propertyValues {FloatData [0.4,0.4]}}
     )
    )
    Mesh Scale=&MakeVector(
     X=&SelectFloatFromArray(
      Direct Array Index=$Particles.PartitionParticles.Index In Partition
      Float Selection Array=DataInterface() {propertyValues {FloatData [0.7,0.7]}}
     )
     Y=0.7
     Z=0.5
    )
    Mesh Scale Mode=::NewEnumerator4
    Position Offset:NiagaraPosition=$User.Position Offset
    UsePositionOffset=true
    Write Material Random=true
   )
   &AddVelocity(Velocity=<0 0 20>) {enabled=false}
   &InitialMeshOrientation(Mesh Orientation Mode=::NewEnumerator2, Use Rotation Vector=false)
   &CameraOffset(Camera Offset Amount=&SelectFloatFromArray(
     Direct Array Index=$Particles.PartitionParticles.Index In Partition
     Float Selection Array=DataInterface() {propertyValues {FloatData [0,1]}}
    ))
  ]
  particleUpdateScript [
   &ParticleState(Kill Particles When Lifetime Has Elapsed=false)
   &SkeletalMeshLocation(
    Bone / Socket Index=$Particles.PartitionParticles.Partition
    Bone Sampling Mode=::NewEnumerator4
    MaskSpawnGroup=false
    Sampled Position Offset:NiagaraPosition=$User.Position Offset
    Skeletal Mesh=$User.Skeletal Mesh
   )
   &ScaleMeshSize(Scale Factor=&Multiply_VectorByFloat(
     Vector=&MakeVector(
      X=1
      Y=1
      Z=&FloatFromCurve(
       FloatCurve=DataInterface() {propertyValues {Curve {keys [
           {arriveTangent=7.7539, interpMode=RCIM_Cubic, leaveTangent=7.7539, tangentMode=RCTM_User, time=0, value=0.2}
           {arriveTangent=0.0586, interpMode=RCIM_Cubic, leaveTangent=0.0586, tangentMode=RCTM_User, time=0.4, value=1.4}
          ]}}}
       CurveIndex=$Particles.Age
      )
     )
     Float=$User.EffectScale
    ))
   &UpdateMeshOrientation(Rotation Coordinate Space=::NewEnumerator0, Rotation Rate=0, Rotation Vector=<0 0 1>) {enabled=false}
   &DynamicMaterialParameters(
    Index 0 Param 1=HlslExpression() {hlslExpression=Emitter.IncomingSpectrumData.x}
    Index 0 Param 2=HlslExpression() {hlslExpression=Emitter.IncomingSpectrumData.y}
    Index 0 Param 3=HlslExpression() {hlslExpression=Emitter.IncomingSpectrumData.z}
    Index 0 Param 4=HlslExpression() {hlslExpression=Emitter.IncomingSpectrumData.a}
    Module.Param0WriteEnabled=true
    Module.Param1WriteEnabled=true
    Module.Param2WriteEnabled=true
    Module.Param3WriteEnabled=true
   )
   &SkeletalMeshLocation(
    Bone / Socket Index=1
    Bone Sampling Mode=::NewEnumerator4
    MaskSpawnGroup=false
    Orientation Sampling=::NewEnumerator1
    Position Sampling=::NewEnumerator1
    Skeletal Mesh=$User.Skeletal Mesh
    Write To Particle Position=false
   ) {enabled=false, module=SkeletalMeshLocation001}
   &SolveForcesAndVelocity(Clamp Velocity=false, Limit Acceleration=false)
  ]
  rendererClasses [&NiagaraMeshRendererProperties]
  renderers [{rendererClass=&NiagaraMeshRendererProperties, rendererIndex=0}]
 }
}
summary {
 systemName=NS_Emote_StarStray
 emitters [
  {emitterName=Volume_Bar, rendererClasses [&NiagaraMeshRendererProperties]}
  {emitterName=Disks, rendererClasses [&NiagaraMeshRendererProperties]}
  {emitterName=CD, rendererClasses [&NiagaraSpriteRendererProperties, &NiagaraLightRendererProperties]}
  {emitterName=GlitchySquare, rendererClasses [&NiagaraSpriteRendererProperties]}
  {emitterName=GlitchySquareBurst, rendererClasses [&NiagaraSpriteRendererProperties]}
  {emitterName=Flash_Flare, rendererClasses [&NiagaraSpriteRendererProperties]}
  {emitterName=Trails, rendererClasses [&NiagaraRibbonRendererProperties]}
  {emitterName=Notes_Line, rendererClasses [&NiagaraRibbonRendererProperties]}
  {emitterName=RainbowFlare, rendererClasses [&NiagaraSpriteRendererProperties]}
  {bEnabled=false, emitterName=FN_DebugGnomon, rendererClasses [&NiagaraMeshRendererProperties]}
  {emitterName=Anticipation_Square, rendererClasses [&NiagaraSpriteRendererProperties]}
  {emitterName=Trails001, rendererClasses [&NiagaraRibbonRendererProperties]}
 ]
 userVariables [
  User.bIsFrontEnd {type=NiagaraBool}
  User.bIsFrontEndPreview {type=NiagaraBool}
  User.bIsInWater {type=NiagaraBool}
  User.EffectScale {type=NiagaraFloat, defaultValue=0.25, description=""}
  User.EffectSpawnScale {type=NiagaraFloat, defaultValue=10, description=""}
  User.ForwardSpeed {type=NiagaraFloat, defaultValue=0, description=""}
  User.LoopDelay {type=NiagaraFloat, defaultValue=0.8, description=""}
  User.Position Offset {type=NiagaraPosition, description="", defaultValue {}}
  User.RibbonIDOffset {type=NiagaraInt32}
  User.Skeletal Mesh {
   type=NiagaraDataInterfaceSkeletalMesh
   defaultValue {
    struct=NiagaraExt_VariableValue_DataInterface
    value {dataInterface=&self:NiagaraDataInterfaceSkeletalMesh_0}
   }
  }
 ]
}
